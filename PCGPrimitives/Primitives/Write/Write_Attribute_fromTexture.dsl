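; PCG primitive graph that writes a named attribute, Density by default, onto its input from a texture, a radial gradient unless overridden.
; The input is tagged "Input:0", then flows through the shared input-management subgraph, a texture-to-attribute core subgraph and an output-preparation subgraph.
; Parameters pass texture, transform, channel mask, normalize and remap ranges to the core.
; Stage graphs are exposed as parameters, making input processing, fallback, core and output logic swappable.
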(graph Write_Attribute_fromTexture
  (params (InputsProcessingGraph :type (SoftObject /Script/PCG.PCGGraphInterface) :default /PCGPrimitives/Primitives/_Template_/Primitives_InputsProcessing.Primitives_InputsProcessing) (InputsFallbackGraph :type (SoftObject /Script/PCG.PCGGraphInterface) :default /PCGPrimitives/Primitives/_Template_/Primitives_InputsFallback.Primitives_InputsFallback) (CoreProcessGraph :type (SoftObject /Script/PCG.PCGGraphInterface) :default /PCGPrimitives/Primitives/Write/Subgraphs/Write_Attribute_fromTexture_CoreProcess.Write_Attribute_fromTexture_CoreProcess) (PrepareOutputsGraph :type (SoftObject /Script/PCG.PCGGraphInterface) :default /PCGPrimitives/Primitives/_Template_/Primitives_PrepareOutputs.Primitives_PrepareOutputs) (Attribute :type String :default "Density") (Normalize :type Bool :default true) (RemapValueInLow :type Double) (RemapValueInHigh :type Double :default 1.0) (RemapValueOutLow :type Double) (RemapValueOutHigh :type Double :default 1.0) (TextureTransform :type Transform :default (struct :Scale3D (5e+02 5e+02 5e+02))) (Texture :type (SoftObject /Script/Engine.Texture2D) :default /PCGPrimitives/Primitives/Write/Ressources/T_gradient_radial.T_gradient_radial) (Mask :type (Enum /Script/GeometryScriptingEditor.EGeometryScriptRGBAChannel)))
  (inputs (PrimaryInput :type Any))
  (bind AddTags_1 (Add%20Tags :TagsToAdd "Input:0"))
  ;[0 0 0 0.5 32] Inputs
  (bind Subgraph_0 (Subgraph :Graph /PCGPrimitives/Primitives/_Template_/Subgraphs/Primitives_InputsManagement.Primitives_InputsManagement :InputsFallbackGraph InputsFallbackGraph :InputsProcessingGraph InputsProcessingGraph))
  ;[0 0 0 0.5 32] Core Process
  (bind Subgraph_3 (Subgraph :Graph /PCGPrimitives/Primitives/Write/Subgraphs/Write_Attribute_fromTexture_CoreProcess.Write_Attribute_fromTexture_CoreProcess :Attribute Attribute :Mask Mask :Normalize Normalize :Remap_value_in_high RemapValueInHigh :Remap_value_in_low RemapValueInLow :Remap_value_out_high RemapValueOutHigh :Remap_value_out_low RemapValueOutLow :SubgraphOverride-param CoreProcessGraph :Texture Texture :Texture_transform TextureTransform))
  ;[0 0 0 0.5 32] Prepare Output
  (bind Subgraph_4 (Subgraph :Graph /PCGPrimitives/Primitives/_Template_/Primitives_PrepareOutputs.Primitives_PrepareOutputs :SubgraphOverride-param PrepareOutputsGraph))
  (connect DefaultInputNode PrimaryInput AddTags_1 In)
  (connect AddTags_1 Out Subgraph_0 In)
  (connect Subgraph_0 Out Subgraph_3 In)
  (connect Subgraph_3 Out Subgraph_4 In)
  (output Out Subgraph_4))
