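; PCG primitive graph that writes a distance-from value to an attribute, "Distance" by default, using primary and secondary inputs.
; Both inputs are tagged by index, gathered, then passed through input management, core process and output preparation subgraphs.
; Parameters control maximum distance, attribute math operation, normalization, 2D computation, vector output and remapping.
; Stage graphs are exposed as parameters, making input processing, fallback, core logic and output preparation swappable.
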(graph Write_DistanceFrom
  (params (InputsProcessingGraph :type (SoftObject /Script/PCG.PCGGraphInterface) :default /PCGPrimitives/Primitives/_Template_/Primitives_InputsProcessing.Primitives_InputsProcessing) (InputsFallbackGraph :type (SoftObject /Script/PCG.PCGGraphInterface) :default /PCGPrimitives/Primitives/_Template_/Primitives_InputsFallback.Primitives_InputsFallback) (CoreProcessGraph :type (SoftObject /Script/PCG.PCGGraphInterface) :default /PCGPrimitives/Primitives/Write/Subgraphs/Write_DistanceFrom_CoreProcess.Write_DistanceFrom_CoreProcess) (PrepareOutputsGraph :type (SoftObject /Script/PCG.PCGGraphInterface) :default /PCGPrimitives/Primitives/_Template_/Primitives_PrepareOutputs.Primitives_PrepareOutputs) (MaximumDistance :type Double :default 1e+02) (OutputAttribute :type String :default "Distance") (AttributeOperation :type (Enum /PCGPrimitives/Primitives/Shared/Resources/EPCG_AttributesMathOpsEnum.EPCG_AttributesMathOpsEnum)) (Normalized :type Bool) (ComputeIn2D :type Bool) (OutputDistanceVector :type Bool) (UseRemap :type Bool) (RemapOutputMin :type Double) (RemapOutputMax :type Double :default 1.0) (RemapClampToUnitRange :type Bool) (RemapIgnoreValuesOutsideInputRange :type Bool))
  (inputs (PrimaryInput :type Any) (SecondaryInput :type Any))
  (bind AddTags_1 (Add%20Tags :TagsToAdd "Input:0"))
  (bind AddTags_2 (Add%20Tags :TagsToAdd "Input:1"))
  (bind Gather_0 (Gather))
  ;[0 0 0 0.5 32] Inputs
  (bind Subgraph_0 (Subgraph :Graph /PCGPrimitives/Primitives/_Template_/Subgraphs/Primitives_InputsManagement.Primitives_InputsManagement :InputsFallbackGraph InputsFallbackGraph :InputsProcessingGraph InputsProcessingGraph))
  ;[0 0 0 0.5 32] Core Process
  (bind Subgraph_3 (Subgraph :Graph /PCGPrimitives/Primitives/Write/Subgraphs/Write_DistanceFrom_CoreProcess.Write_DistanceFrom_CoreProcess :AttributeOperation AttributeOperation :ComputeIn2D ComputeIn2D :MaximumDistance MaximumDistance :Normalized Normalized :OuputAttribute OutputAttribute :OutputDistanceVector OutputDistanceVector :Remap_output_max RemapOutputMax :Remap_output_min RemapOutputMin :RemapClampToUnitRange RemapClampToUnitRange :RemapIgnoreValuesOutsideInputRange RemapIgnoreValuesOutsideInputRange :SubgraphOverride-param CoreProcessGraph :UseRemap UseRemap))
  ;[0 0 0 0.5 32] Prepare Output
  (bind Subgraph_4 (Subgraph :Graph /PCGPrimitives/Primitives/_Template_/Primitives_PrepareOutputs.Primitives_PrepareOutputs :SubgraphOverride-param PrepareOutputsGraph))
  (connect DefaultInputNode PrimaryInput AddTags_1 In)
  (connect DefaultInputNode SecondaryInput AddTags_2 In)
  (connect AddTags_1 Out Gather_0 In)
  (connect AddTags_2 Out Gather_0 In2)
  (connect Gather_0 Out Subgraph_0 In)
  (connect Subgraph_0 Out Subgraph_3 In)
  (connect Subgraph_3 Out Subgraph_4 In)
  (output Out Subgraph_4))
