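; Write_Noise is a PCG primitive graph that writes noise to an attribute, using configurable input and output attributes (density by default), noise range and math operation.
; It tags the primary input, then chains templated input-management, noise core-process and output-preparation subgraphs.
; Input processing, input fallback, core and output graphs are exposed as parameters, making each stage swappable.
; The core receives a seed and an optional custom seed source.
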(graph Write_Noise
  (params (InputsProcessingGraph :type (SoftObject /Script/PCG.PCGGraphInterface) :default /PCGPrimitives/Primitives/_Template_/Primitives_InputsProcessing.Primitives_InputsProcessing) (InputsFallbackGraph :type (SoftObject /Script/PCG.PCGGraphInterface) :default /PCGPrimitives/Primitives/_Template_/Primitives_InputsFallback.Primitives_InputsFallback) (CoreProcessGraph :type (SoftObject /Script/PCG.PCGGraphInterface) :default /PCGPrimitives/Primitives/Write/Subgraphs/Write_Noise_CoreProcess.Write_Noise_CoreProcess) (InputAttribute :type String :default "$Density") (PrepareOutputsGraph :type (SoftObject /Script/PCG.PCGGraphInterface) :default /PCGPrimitives/Primitives/_Template_/Primitives_PrepareOutputs.Primitives_PrepareOutputs) (OutputAttribute :type String :default "$Density") (AttributeOperation :type (Enum /PCGPrimitives/Primitives/Shared/Resources/EPCG_AttributesMathOpsEnum.EPCG_AttributesMathOpsEnum)) (NoiseMin :type Double) (NoiseMax :type Double :default 1.0) (Seed :type Int :default -488115970) (HasCustomSeedSource :type Bool) (CustomSeedSource :type String :default "@Last"))
  (inputs (PrimaryInput :type Any))
  (bind AddTags_1 (Add%20Tags :TagsToAdd "Input:0"))
  ;[0 0 0 0.5 32] Inputs
  (bind Subgraph_0 (Subgraph :Graph /PCGPrimitives/Primitives/_Template_/Subgraphs/Primitives_InputsManagement.Primitives_InputsManagement :InputsFallbackGraph InputsFallbackGraph :InputsProcessingGraph InputsProcessingGraph))
  ;[0 0 0 0.5 32] Core Process
  (bind Subgraph_3 (Subgraph :Graph /PCGPrimitives/Primitives/Write/Subgraphs/Write_Noise_CoreProcess.Write_Noise_CoreProcess :AttributeOperation AttributeOperation :CustomSeedSource CustomSeedSource :HasCustomSeedSource HasCustomSeedSource :InputAttribute InputAttribute :NoiseMax NoiseMax :NoiseMin NoiseMin :OuputAttribute OutputAttribute :Seed-param Seed :SubgraphOverride-param CoreProcessGraph))
  ;[0 0 0 0.5 32] Prepare Output
  (bind Subgraph_4 (Subgraph :Graph /PCGPrimitives/Primitives/_Template_/Primitives_PrepareOutputs.Primitives_PrepareOutputs :SubgraphOverride-param PrepareOutputsGraph))
  (connect DefaultInputNode PrimaryInput AddTags_1 In)
  (connect AddTags_1 Out Subgraph_0 In)
  (connect Subgraph_0 Out Subgraph_3 In)
  (connect Subgraph_3 Out Subgraph_4 In)
  (output Out Subgraph_4))
